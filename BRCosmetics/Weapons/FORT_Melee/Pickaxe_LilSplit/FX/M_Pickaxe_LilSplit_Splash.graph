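# Unlit translucent Niagara sprite and mesh material for the LilSplit pickaxe's hand-painted splash effect.
# Time-panned Perlin noise distorts the splash texture, which blends two tint colors multiplied by particle color.
# Normals derived from the splash heightmap sample a matcap, adding a masked highlight to emissive.
# Particle alpha drives smoothstep dissolve opacity with dynamic-parameter softness, depth-faded except at Low quality.
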Material /BRCosmetics/Weapons/FORT_Melee/Pickaxe_LilSplit/FX/M_Pickaxe_LilSplit_Splash
graph {
 properties {
  bUsedWithNiagaraMeshParticles=true
  bUsedWithNiagaraSprites=true
  blendMode=BLEND_Translucent
  pixelDepthOffsetMode=PDOM_Legacy
  shadingModel=MSM_Unlit
  usedShadingModels=Unlit
 }
 nodes [
  n1 = TextureCoordinate()
  n2 = Multiply(A=Time()) {constB=0.6}
  n3 = Multiply(A=TextureSample(UVs=Add(A=@n1, B=@n2)) {samplerType=SAMPLERTYPE_LinearColor, texture=/BRCosmetics/Weapons/FORT_Melee/Pickaxe_RhombCamo/FX/T_Pickaxe_RhombCamo_Perlin}.R) {constB=0.06}
  n4 = TextureObjectParameter() {parameterName=SplashTexture, texture=/BRCosmetics/Weapons/FORT_Melee/Pickaxe_RecordScratch/FX/Textures/T_Water_Splash_Handpainted_Dissolve}
  n5 = TextureSample(UVs=Add(A=@n1, B=@n3), Tex=@n4)
  n6 = ParticleColor()
  n7 = Multiply(
   A=LinearInterpolate(
    A=VectorParameter() {parameterName=ColorA, defaultValue=<rgba 0.5 0.3518 0.162 1>}.RGB
    B=VectorParameter() {parameterName=ColorB, defaultValue=<rgba 0.6875 0.5923 0.2874 1>}.RGB
    Alpha=Multiply(A=@n5.R, B=@n5.R)
   )
   B=@n6.RGB
  )
  n8 = ParticleColor()
  n9 = MaterialFunctionCall!(Height Map=@n4, Normal Map Intensity, Height Map UV Offset, Coordinates, Height Map Channel Selector) {materialFunction=/Engine/Functions/Engine_MaterialFunctions03/Procedurals/NormalFromHeightmap}
  n10 = Add(A=Multiply(A=Transform(None=@n9.Result) {transformType=TRANSFORM_View}) {constB=0.5}) {constB=0.5}
  n11 = TextureSample(UVs=Add(A=ComponentMask(None=@n10) {g=true, r=true}, B=@n3)) {
   samplerType=SAMPLERTYPE_LinearGrayscale
   texture=/BRCosmetics/Athena/Effects/Particles/Skydiving/NS_Contrail_FNCS_S26/T_Contrail_FNCS_S26_Matcap
  }
  n12 = OneMinus(None=@n6.A)
  n13 = MaterialFunctionCall(
   Alpha=@n5.R
   Min=@n12
   Max=Add(
    A=@n12
    B=DynamicParameter() {defaultValue=<rgba 0.1 1 1 1>, paramNames [SmoothstepSoftness, Random, WarpStrength, Panning Speed]}.SmoothstepSoftness
   )
  ) {materialFunction=&SmoothStep}
  n14 = Multiply(A=@n5.R, B=@n13.Result)
  MaterialFunctionCall(Emissive=Add(A=Multiply(A=@n7, B=@n8.RGB), B=Multiply(A=@n11.R, B=@n14))) {materialFunction=&MF_FX_EmissiveAdjustments}
  n15 = DepthFade(Opacity=@n13.Result, FadeDistance=ScalarParameter() {defaultValue=50, parameterName=DepthFade}) {fadeDistanceDefault=50}
  QualitySwitch(Default=@n15, Low=@n13.Result, High=@n15, Medium=@n15, Epic=@n15)
 ]
}
