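# Physics asset for the offhand axe mesh of the LilSplit pickaxe cosmetic, covering its rear drip bones.
# A small capsule body sits on the base drip bone and a larger, offset capsule on the dynamic drip bone.
# One constraint joins the two bodies with narrow swing limits and a zero-degree twist limit, suggesting a slight dangling sway.
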PhysicsAsset /BRCosmetics/Weapons/FORT_Melee/Pickaxe_LilSplit/Meshes/LilSplitOffhand_Axe_Physics
bodies {
 base_drip_bck [{shapeType=Capsule, length=0.505, radius=0.505, shapeName=base_drip_bck_capsule, rotation=<rot 0 -90 0>}]
 dyn_drip_bck [{shapeType=Capsule, length=2, radius=1, shapeName=dyn_drip_bck_capsule, center=<2 2 1.8>, rotation=<rot 90 0 0>}]
}
constraints [{bone1Name=dyn_drip_bck, bone2Name=base_drip_bck, swing1LimitDegrees=18, swing2LimitDegrees=28, twistLimitDegrees=0}]
